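; Game state component that manages icy ground patches and the shared frosty wisp Niagara effects on them.
; New locations join a nearby patch actor or spawn a new one, generating patches via an optional EQS query, while reverse iteration prunes invalid actors.
; Patches can share a particle when close with similar ground normals.
; Reference counts destroy unused effects, resetting their array slots for reuse without shifting indices.
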(fn AddIcyPatchAtLocation
  (params (PatchLocation :type Vector) (EQS%20Query :type Env%20Query%20Object%20Reference) (PatchCombinationRadius :type Float%20%28double-precision%29))
  ;[0.04011 0.15 0.1443 0.5] Try to find already spawned patch near location that can be added to
  (bind (_arrayindex _arrayelement) (Utilities|Array|ReverseforEachLoop (Variables|Default|GetSpawnedPatchActors))
    (:LoopBody
      (Utilities|IsValid _arrayelement
        (:Is%20Valid
          (bind _bshouldadd (PatchHelpers|ShouldAddLocationtoPatch _arrayelement PatchLocation PatchCombinationRadius))
          (if _bshouldadd
            (Utilities|IsValid EQS%20Query
              (:Is%20Valid
                (Class|BIcyPatch|GeneratePatchesatLocation _arrayelement PatchLocation EQS%20Query))
              (:Is%20Not%20Valid
                (Class|BIcyPatch|GeneratePatchatLocation _arrayelement PatchLocation)))
            (return)))
        (:Is%20Not%20Valid
          (Utilities|Array|RemoveIndex (Variables|Default|GetSpawnedPatchActors) _arrayindex))))
    (:Completed
      ;[0.0847 0.15 0.04961 0.5] Spawn new patch if none can be combined with
      (bind _returnvalue (Game|SpawnActorfromClass (Variables|Default|GetIcyPatchClass) (Math|Transform|MakeTransform PatchLocation) "Undefined" "MultiplyWithRoot" 0 0))
      (Utilities|IsValid _returnvalue
        (:Is%20Valid
          (Utilities|Array|Add (Variables|Default|GetSpawnedPatchActors) _returnvalue)
          (Utilities|IsValid EQS%20Query
            (:Is%20Valid
              (Class|BIcyPatch|GeneratePatchesatLocation _returnvalue PatchLocation EQS%20Query))
            (:Is%20Not%20Valid
              (Class|BIcyPatch|GeneratePatchatLocation _returnvalue PatchLocation))))))) ; Reversed to allow for safe removal \nof invalid patch actors
)

(fn AddParticle
  (params (NewParticle :type Icy%20Patch%20Particle%20Data%20Structure%20%28by%20ref%29))
  (bind (_location_4_0b01d8c84ddd74cd5df78d9cbe87e092 _normal_5_c845a950432bcee1a7cfb3bab26e78f0 _patchcount_17_00e3f4e549ca99fc81c81586268f2443) (Utilities|Struct|BreakIcyPatchParticleData NewParticle))
  (bind _icypatchparticles (Variables|Default|GetIcyPatchParticles))
  ;[0.03726 0.15 0.02158 0.5] Create new particle
  (bind _returnvalue (Niagara|SpawnSystematLocation "/TeaCakeIcyShins/Assets/FX/NS_IceGauntlets_FrostyWisps.NS_IceGauntlets_FrostyWisps" _location_4_0b01d8c84ddd74cd5df78d9cbe87e092 (Math|Rotator|MakeRotfromZ _normal_5_c845a950432bcee1a7cfb3bab26e78f0)))
  ;;[0.05416 0.15 0.1224 0.5] If there are any particles without a patch, reuse that index
  (for (_array_element _array_index) _icypatchparticles
    (if (<= (.PatchCount_17_00E3F4E549CA99FC81C81586268F2443 (Utilities|Struct|BreakIcyPatchParticleData _array_element)) 0)
      (Utilities|Array|SetArrayElem _icypatchparticles _array_index (Utilities|Struct|MakeIcyPatchParticleData _location_4_0b01d8c84ddd74cd5df78d9cbe87e092 _normal_5_c845a950432bcee1a7cfb3bab26e78f0 _patchcount_17_00e3f4e549ca99fc81c81586268f2443 _returnvalue))
      (return _array_index)))
  ;;[0.0096 0.15 0.02539 0.5] Add new particle if none are free
  (bind _returnvalue_2 (Utilities|Array|Add _icypatchparticles (Utilities|Struct|MakeIcyPatchParticleData _location_4_0b01d8c84ddd74cd5df78d9cbe87e092 _normal_5_c845a950432bcee1a7cfb3bab26e78f0 _patchcount_17_00e3f4e549ca99fc81c81586268f2443 _returnvalue)))
  (return _returnvalue_2))

(fn CanPatchUseParticle
  (params (ParticleIndex :type Integer) (Patch :type Icy%20Patch%20Data%20Structure%20%28by%20ref%29) (PatchGroundHitResult :type Hit%20Result%20Structure%20%28by%20ref%29))
  (bind (_location_4_0b01d8c84ddd74cd5df78d9cbe87e092 _normal_5_c845a950432bcee1a7cfb3bab26e78f0) (Utilities|Struct|BreakIcyPatchParticleData (Utilities|Array|Get(aref) (Variables|Default|GetIcyPatchParticles) ParticleIndex)))
  (bind (_bblockinghit _binitialoverlap _time _distance _location _impactpoint _normal) (Collision|BreakHitResult PatchGroundHitResult))
  (bind _icypatchparticles (Variables|Default|GetIcyPatchParticles))
  ;[0.04758 0.06675 0.15 0.5] Allow patch to use particle if it is close enough and has a similar enough ground normal
  (if (Utilities|Array|IsValidIndex _icypatchparticles ParticleIndex)
    (return (and (< (Math|Vector|DistanceSquared(Vector) (Utilities|Struct|BreakIcyPatchData Patch) _location_4_0b01d8c84ddd74cd5df78d9cbe87e092) (Variables|Default|GetParticleMaxRadiusSquared)) (> (Math|Vector|DotProduct _normal_5_c845a950432bcee1a7cfb3bab26e78f0 _normal) (Variables|Default|GetMinParticleNormalDifference))))
    (else
      (return false))))

(fn ModifyParticlePatchCount
  (params (ParticleIndex :type Integer) (CountChange :type Integer))
  (bind _icypatchparticles (Variables|Default|GetIcyPatchParticles))
  (bind _output (Utilities|Array|Get(aref) _icypatchparticles ParticleIndex))
  (bind (_location_4_0b01d8c84ddd74cd5df78d9cbe87e092 _normal_5_c845a950432bcee1a7cfb3bab26e78f0 _patchcount_17_00e3f4e549ca99fc81c81586268f2443 _tempsystemcomponent_19_8aa4717846a2b685b263df818f0df83a) (Utilities|Struct|BreakIcyPatchParticleData _output))
  (bind _returnvalue (+ _patchcount_17_00e3f4e549ca99fc81c81586268f2443 CountChange))
  ;[0.15 0.1461 0.02805 0.5] Update how many patches are usuing a particle. Once a particle is no longer being used, remove it
  (if (Utilities|Array|IsValidIndex _icypatchparticles ParticleIndex)
    (if (> _returnvalue 0)
      (Utilities|Array|SetArrayElem _icypatchparticles ParticleIndex (Utilities|Struct|MakeIcyPatchParticleData _location_4_0b01d8c84ddd74cd5df78d9cbe87e092 _normal_5_c845a950432bcee1a7cfb3bab26e78f0 _returnvalue _tempsystemcomponent_19_8aa4717846a2b685b263df818f0df83a))
      (else
        (Default|RemoveParticle ParticleIndex)))))

(fn RemoveParticle
  (params (ParticleIndex :type Integer))
  (bind _icypatchparticles (Variables|Default|GetIcyPatchParticles))
  (bind _output (Utilities|Array|Get(aref) _icypatchparticles ParticleIndex))
  (bind (_location_4_0b01d8c84ddd74cd5df78d9cbe87e092 _normal_5_c845a950432bcee1a7cfb3bab26e78f0 _patchcount_17_00e3f4e549ca99fc81c81586268f2443 _tempsystemcomponent_19_8aa4717846a2b685b263df818f0df83a) (Utilities|Struct|BreakIcyPatchParticleData _output))
  ;[0.15 0.03809 0.02805 0.5] Remove particle
  (if (Utilities|Array|IsValidIndex _icypatchparticles ParticleIndex)
    (Utilities|IsValid _tempsystemcomponent_19_8aa4717846a2b685b263df818f0df83a
      (:Is%20Valid
        (Components|DestroyComponent _tempsystemcomponent_19_8aa4717846a2b685b263df818f0df83a)
        (Utilities|Array|SetArrayElem _icypatchparticles ParticleIndex (Utilities|Struct|MakeIcyPatchParticleData "0.000000,0.000000,0.000000")) ; Free this array element by reseting its data
      ))))
